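# Physics asset for the dynamic main belt on the Banner female medium Commando body, likely for secondary belt motion.
# Capsule bodies on pelvis, belt base, main belt and rear-right belt bones form a constraint chain in that order.
# Only the main-belt-to-base joint is tightly limited in swing and twist; the pelvis-to-base and main-to-rear-right joints leave all angular axes free.
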PhysicsAsset /BRCosmetics/Characters/Player/Female/Medium/Bodies/F_MED_Commando_Banner/Meshes/F_MED_Banner_BeltMain_Dyn_Physics
bodies {
 dyn_belt_main [{shapeType=Capsule, length=11, radius=3.5, center=<-2.4308 0 0>, rotation=<rot 90 0 0>}]
 dyn_belt_main_base [{shapeType=Capsule, length=0.505, radius=0.505, rotation=<rot 0 -90 0>}]
 dyn_belt_rear_r [{
   shapeType=Capsule
   length=6.4261
   radius=3.8234
   shapeName=dyn_belt_rear_r_capsule
   center=<0.2042 0.0263 -0.4432>
   rotation=<rot 67.2001 143.9176 -128.3267>
  }]
 pelvis [{shapeType=Capsule, length=1.3181, radius=5.8688, center=<0.7703 -1.5474 1.1758>, rotation=<rot 13.2873 -89.9999 3.3043>}]
}
constraints [
 {
  bone1Name=dyn_belt_main_base
  bone2Name=pelvis
  swing1LimitDegrees=45
  swing1Motion=Free
  swing2LimitDegrees=45
  swing2Motion=Free
  twistLimitDegrees=45
  twistMotion=Free
 }
 {bone1Name=dyn_belt_main, bone2Name=dyn_belt_main_base, swing1LimitDegrees=2, swing2LimitDegrees=10, twistLimitDegrees=2}
 {
  bone1Name=dyn_belt_rear_r
  bone2Name=dyn_belt_main
  swing1LimitDegrees=45
  swing1Motion=Free
  swing2LimitDegrees=45
  swing2Motion=Free
  twistLimitDegrees=45
  twistMotion=Free
 }
]
