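# Animation Blueprint for the female Commando Banner body that copies the parent mesh pose and adds secondary motion to belt, kneepad, thigh, shoelace and leg-flap bones.
# RBF Pose Drivers and constraints make belt and kneepad follow the legs, then Anim Dynamics swings laces and flaps.
# A spine-based RigidBody belt simulation is cached separately, bypassed while crouching.
# EventGraph maps bone rotations and fall speed to alphas, including left thigh volume correction.
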AnimBlueprint /BRCosmetics/Characters/Player/Female/Medium/Bodies/F_MED_Commando_Banner/Meshes/F_MED_Banner_Skeleton_AnimBP
target_skeleton=/Game/Characters/Player/Female/Medium/Bodies/F_MED_Commando_Banner/Meshes/F_MED_Banner_Skeleton
graphs [
 AnimGraph {nodes [
   Anim Dynamics - Bone: dyn_belt_main(Alpha=1, ExternalForce="0,0,0") {
    angularDampingOverride=1
    angularSpringConstant=1000
    bAngularSpring=true
    bGravityOverrideInSimSpace=true
    bOverrideAngularDamping=true
    gravityScale=0.35
    alphaBoolBlend {blendOption=Linear}
    boundBone {boneName=dyn_belt_main}
    componentAppliedLinearAccClamp=<1000 1000 1000>
    componentLinearAccScale=<0.2 0.2 0.2>
    componentLinearVelScale=<0.2 0.2 0.2>
    physicsBodyDefinitions [{
      sphereCollisionRadius=0
      boundBone {boneName=dyn_belt_main}
      boxExtents=<5 5 5>
      constraintSetup {angularLimitsMax=<0 40 0>, angularLimitsMin=<0 0 0>, angularTarget=<1 0 0>}
      localJointOffset=<0 0 -1>
     }]
   }
   "Transform (Modify) Bone - Bone: dyn_belt_main_base"(Alpha=Get jump_alpha() -> (jump_alpha:double), Translation="0,0,0", Rotation="-10,0,0", Scale="1,1,1") {
    rotationMode=BMM_Additive
    rotationSpace=BCS_BoneSpace
    translationMode=BMM_Additive
    translationSpace=BCS_BoneSpace
    alphaBoolBlend {blendOption=Linear}
    boneToModify {boneName=dyn_belt_main_base}
    rotation {pitch=-10}
   }
   Output Pose(Result=Blend Poses by bool(BlendPose_0=Use cached pose 'CopyPoseBaseCache'(), BlendPose_1=Use cached pose 'dynamicBeltPose'(), bActiveValue=Get bIsCrouching()) {blendType=Linear}) {layerGroup=None}
   n1 = Pose Driver - Source: thigh_r!(SourcePose:Pose Link Structure=Copy Pose From Mesh() {bUseAttachedParent=true}) -> (Pose:Pose Link Structure) {
    is=title AnimGraphNode_PoseDriver
    poseAsset=&/M_MED_Banner_Leg_RBF_poseAsset
    onlyDriveBones [{boneName=dyn_belt_front_r}, {boneName=dyn_belt_rear_r}, {boneName=dyn_belt_side_a_r}, {boneName=dyn_belt_side_b_r}]
    poseTargets [
     {
      bApplyCustomCurve=false
      bIsHidden=false
      distanceMethod=DefaultMethod
      drivenName=default
      functionType=DefaultFunction
      targetScale=0
      boneTransforms [{targetRotation=<rot 4.0665 178.2257 0.0403>, targetTranslation=<-1.0853 -0.2109 8.7053>}]
      customCurve {defaultValue=3.4028234663852886e+38, postInfinityExtrap=RCCE_Constant, preInfinityExtrap=RCCE_Constant, keys []}
      targetRotation=<rot 0 0 0>
     }
     {
      bApplyCustomCurve=false
      bIsHidden=false
      distanceMethod=DefaultMethod
      drivenName=upleg_upMid_r
      functionType=DefaultFunction
      targetScale=0.5468
      boneTransforms [{targetRotation=<rot 2.6435 128.1529 -3.0915>, targetTranslation=<-1.0853 -0.2109 8.7053>}]
      customCurve {defaultValue=3.4028234663852886e+38, postInfinityExtrap=RCCE_Constant, preInfinityExtrap=RCCE_Constant, keys []}
      targetRotation=<rot 0 0 0>
     }
     {
      bApplyCustomCurve=false
      bIsHidden=false
      distanceMethod=DefaultMethod
      drivenName=upleg_upMax_r
      functionType=DefaultFunction
      targetScale=0.5468
      boneTransforms [{targetRotation=<rot -0.666 78.2476 -4.0119>, targetTranslation=<-1.0853 -0.2109 8.7053>}]
      customCurve {defaultValue=3.4028234663852886e+38, postInfinityExtrap=RCCE_Constant, preInfinityExtrap=RCCE_Constant, keys []}
      targetRotation=<rot 0 0 0>
     }
     {
      bApplyCustomCurve=false
      bIsHidden=false
      distanceMethod=DefaultMethod
      drivenName=upleg_side_r
      functionType=DefaultFunction
      targetScale=1
      boneTransforms [{targetRotation=<rot 54.0665 178.2784 0.0685>, targetTranslation=<-1.0853 -0.2109 8.7053>}]
      customCurve {defaultValue=3.4028234663852886e+38, postInfinityExtrap=RCCE_Constant, preInfinityExtrap=RCCE_Constant, keys []}
      targetRotation=<rot 0 0 0>
     }
     {
      bApplyCustomCurve=false
      bIsHidden=false
      distanceMethod=DefaultMethod
      drivenName=upleg_upSideMid_r
      functionType=DefaultFunction
      targetScale=0.5468
      boneTransforms [{targetRotation=<rot 27.6037 126.6792 -3.4853>, targetTranslation=<-1.0853 -0.2109 8.7053>}]
      customCurve {defaultValue=3.4028234663852886e+38, postInfinityExtrap=RCCE_Constant, preInfinityExtrap=RCCE_Constant, keys []}
      targetRotation=<rot 0 0 0>
     }
     {
      bApplyCustomCurve=false
      bIsHidden=false
      distanceMethod=DefaultMethod
      drivenName=upleg_upSideMax_r
      functionType=DefaultFunction
      targetScale=0.5468
      boneTransforms [{targetRotation=<rot 24.2689 76.3889 -4.4013>, targetTranslation=<-1.0853 -0.2109 8.7053>}]
      customCurve {defaultValue=3.4028234663852886e+38, postInfinityExtrap=RCCE_Constant, preInfinityExtrap=RCCE_Constant, keys []}
      targetRotation=<rot 0 0 0>
     }
    ]
    rBFParams {radius=45}
    sourceBones [{boneName=thigh_r}]
   }
   n2 = Pose Driver - Source: thigh_l!(SourcePose:Pose Link Structure=@n1) -> (Pose:Pose Link Structure) {
    is=title AnimGraphNode_PoseDriver
    poseAsset=&M_MED_Banner_Leg_RBF_poseAsset
    onlyDriveBones [{boneName=dyn_belt_front_l}]
    poseTargets [
     {
      bApplyCustomCurve=false
      bIsHidden=false
      distanceMethod=DefaultMethod
      drivenName=default
      functionType=DefaultFunction
      targetScale=0
      boneTransforms [{targetRotation=<rot 4.0665 -1.7743 0.0403>, targetTranslation=<-1.0855 -0.2109 -8.7053>}]
      customCurve {defaultValue=3.4028234663852886e+38, postInfinityExtrap=RCCE_Constant, preInfinityExtrap=RCCE_Constant, keys []}
      targetRotation=<rot 0 0 0>
     }
     {
      bApplyCustomCurve=false
      bIsHidden=false
      distanceMethod=DefaultMethod
      drivenName=upleg_upMax_l
      functionType=DefaultFunction
      targetScale=1
      boneTransforms [{targetRotation=<rot -0.666 -101.7524 -4.0119>, targetTranslation=<-1.0855 -0.2109 -8.7053>}]
      customCurve {defaultValue=3.4028234663852886e+38, postInfinityExtrap=RCCE_Constant, preInfinityExtrap=RCCE_Constant, keys []}
      targetRotation=<rot 0 0 0>
     }
    ]
    rBFParams {radius=50}
    sourceBones [{boneName=thigh_l}]
   }
   n3 = "Constraint - dyn_belt_front_r (P)"(ComponentPose=Local To Component(LocalPose=@n2), Alpha=0.4, ConstraintWeights_0=1) {
    alphaBoolBlend {blendOption=Linear}
    boneToModify {boneName=dyn_belt_front_r}
    constraintSetup [{targetBone {boneName=thigh_r}}]
    constraintWeights [1]
   }
   n4 = "Constraint - dyn_belt_side_a_r (P)"(ComponentPose=@n3, Alpha=0.4, ConstraintWeights_0=1) {
    alphaBoolBlend {blendOption=Linear}
    boneToModify {boneName=dyn_belt_side_a_r}
    constraintSetup [{targetBone {boneName=thigh_r}}]
    constraintWeights [1]
   }
   n5 = Component To Local(ComponentPose=@n4)
   n6 = Pose Driver - Source: thigh_l(SourcePose=Copy Pose From Mesh() {bUseAttachedParent=true}) {
    poseAsset=&M_MED_Banner_Leg_RBF_poseAsset
    onlyDriveBones [{boneName=dyn_belt_front_l}]
    poseTargets [
     {
      bApplyCustomCurve=false
      bIsHidden=false
      distanceMethod=DefaultMethod
      drivenName=default
      functionType=DefaultFunction
      targetScale=0
      boneTransforms [{targetRotation=<rot 4.0665 -1.7743 0.0403>, targetTranslation=<-1.0855 -0.2109 -8.7053>}]
      customCurve {defaultValue=3.4028234663852886e+38, postInfinityExtrap=RCCE_Constant, preInfinityExtrap=RCCE_Constant, keys []}
      targetRotation=<rot 0 0 0>
     }
     {
      bApplyCustomCurve=false
      bIsHidden=false
      distanceMethod=DefaultMethod
      drivenName=upleg_upMax_l
      functionType=DefaultFunction
      targetScale=1
      boneTransforms [{targetRotation=<rot -0.666 -101.7524 -4.0119>, targetTranslation=<-1.0855 -0.2109 -8.7053>}]
      customCurve {defaultValue=3.4028234663852886e+38, postInfinityExtrap=RCCE_Constant, preInfinityExtrap=RCCE_Constant, keys []}
      targetRotation=<rot 0 0 0>
     }
    ]
    rBFParams {radius=50}
    sourceBones [{boneName=thigh_l}]
   }
   n7 = "Transform (Modify) Bone - Bone: dyn_belt_front_r"(ComponentPose=Local To Component(LocalPose=@n6), Alpha=1, Translation="0,0,10", Rotation="0,0,0", Scale="1,1,1") {
    alpha=0
    rotationMode=BMM_Additive
    rotationSpace=BCS_BoneSpace
    translationMode=BMM_Additive
    translationSpace=BCS_BoneSpace
    alphaBoolBlend {blendOption=Linear}
    boneToModify {boneName=dyn_belt_front_r}
    translation {z=10}
   }
   n8 = "Transform (Modify) Bone - Bone: dyn_belt_side_a_r"(ComponentPose=@n7, Alpha=1, Translation="-2,0,12", Rotation="15,0,15", Scale="1,1,1") {
    alpha=0
    rotationMode=BMM_Additive
    rotationSpace=BCS_BoneSpace
    translationMode=BMM_Additive
    translationSpace=BCS_BoneSpace
    alphaBoolBlend {blendOption=Linear}
    boneToModify {boneName=dyn_belt_side_a_r}
    rotation {pitch=15, roll=15}
    translation {x=-2, z=12}
   }
   n9 = "Transform (Modify) Bone - Bone: dyn_belt_side_b_r"(ComponentPose=@n8, Alpha=1, Translation="-1,0,7", Rotation="0,0,15", Scale="1,1,1") {
    alpha=0
    rotationMode=BMM_Additive
    rotationSpace=BCS_BoneSpace
    translationMode=BMM_Additive
    translationSpace=BCS_BoneSpace
    alphaBoolBlend {blendOption=Linear}
    boneToModify {boneName=dyn_belt_side_b_r}
    rotation {roll=15}
    translation {x=-1, z=7}
   }
   n10 = "Transform (Modify) Bone - Bone: dyn_belt_rear_r"(ComponentPose=@n9, Alpha=1, Translation="0,0,3", Rotation="-15,0,0", Scale="1,1,1") {
    alpha=0
    rotationMode=BMM_Additive
    rotationSpace=BCS_BoneSpace
    translationMode=BMM_Additive
    translationSpace=BCS_BoneSpace
    alphaBoolBlend {blendOption=Linear}
    boneToModify {boneName=dyn_belt_rear_r}
    rotation {pitch=-15}
    translation {z=3}
   }
   n11 = "Transform (Modify) Bone - Bone: dyn_belt_front_r"(ComponentPose=@n10, Alpha=1, Translation="0,0,0", Rotation="0,0,0", Scale="1,1,1") {
    alpha=0
    rotationMode=BMM_Additive
    rotationSpace=BCS_BoneSpace
    translationMode=BMM_Additive
    translationSpace=BCS_BoneSpace
    alphaBoolBlend {blendOption=Linear}
    boneToModify {boneName=dyn_belt_front_r}
   }
   n12 = "Transform (Modify) Bone - Bone: dyn_belt_side_a_r"(ComponentPose=@n11, Alpha=1, Translation="0,0,0", Rotation="0,0,0", Scale="1,1,1") {
    alpha=0
    rotationMode=BMM_Additive
    rotationSpace=BCS_BoneSpace
    translationMode=BMM_Additive
    translationSpace=BCS_BoneSpace
    alphaBoolBlend {blendOption=Linear}
    boneToModify {boneName=dyn_belt_side_a_r}
   }
   n13 = "Transform (Modify) Bone - Bone: dyn_belt_side_b_r"(ComponentPose=@n12, Alpha=1, Translation="0,0,0", Rotation="0,0,0", Scale="1,1,1") {
    alpha=0
    rotationMode=BMM_Additive
    rotationSpace=BCS_BoneSpace
    translationMode=BMM_Additive
    translationSpace=BCS_BoneSpace
    alphaBoolBlend {blendOption=Linear}
    boneToModify {boneName=dyn_belt_side_b_r}
   }
   n14 = "Transform (Modify) Bone - Bone: dyn_belt_rear_r"(ComponentPose=@n13, Alpha=1, Translation="0,0,0", Rotation="0,0,0", Scale="1,1,1") {
    alpha=0
    rotationMode=BMM_Additive
    rotationSpace=BCS_BoneSpace
    translationMode=BMM_Additive
    translationSpace=BCS_BoneSpace
    alphaBoolBlend {blendOption=Linear}
    boneToModify {boneName=dyn_belt_rear_r}
   }
   n15 = Component To Local(ComponentPose=@n14)
   n16 = "Constraint - dyn_kneepad (PP)"!(
    ComponentPose=Local To Component(LocalPose=Blend Poses by bool(BlendPose_0=@n15, BlendPose_1=@n5) {blendType=Linear})
    Alpha=1
    ConstraintWeights_0=0.6
    ConstraintWeights_1:float=0.4
   ) -> (Pose) {
    alphaBoolBlend {blendOption=Linear}
    boneToModify {boneName=dyn_kneepad}
    constraintSetup [{targetBone {boneName=thigh_l}}, {targetBone {boneName=calf_l}}]
    constraintWeights [1,1]
   }
   n17 = "Transform (Modify) Bone - Bone: dyn_kneepad"(ComponentPose=@n16, Alpha=1, Translation="0,-3,0", Rotation="0,0,0", Scale="1,1,1.2") {
    alpha=0
    rotationMode=BMM_Additive
    rotationSpace=BCS_BoneSpace
    scaleMode=BMM_Replace
    scaleSpace=BCS_BoneSpace
    translationMode=BMM_Additive
    translationSpace=BCS_BoneSpace
    alphaBoolBlend {blendOption=Linear}
    boneToModify {boneName=dyn_kneepad}
    scale {z=1.2}
    translation {y=-3}
   }
   n18 = "Constraint - dyn_belt_front_base (PP)"!(ComponentPose=@n17, Alpha=1, ConstraintWeights_0=0.5, ConstraintWeights_1:float=0.5) -> (Pose) {
    alphaBoolBlend {blendOption=Linear}
    boneToModify {boneName=dyn_belt_front_base}
    constraintSetup [{targetBone {boneName=dyn_belt_front_r}}, {targetBone {boneName=dyn_belt_front_l}}]
    constraintWeights [1,1]
   }
   n19 = "Constraint - thigh_volume_l (R)"(ComponentPose=@n18, Alpha=0.5, ConstraintWeights_0=1) {
    alphaBoolBlend {blendOption=Linear}
    boneToModify {boneName=thigh_volume_l}
    constraintSetup [{transformType=Rotation, targetBone {boneName=thigh_l}}]
    constraintWeights [1]
   }
   n20 = "Constraint - thigh_volume_r (R)"(ComponentPose=@n19, Alpha=0.5, ConstraintWeights_0=1) {
    alphaBoolBlend {blendOption=Linear}
    boneToModify {boneName=thigh_volume_r}
    constraintSetup [{transformType=Rotation, targetBone {boneName=thigh_r}}]
    constraintWeights [1]
   }
   n21 = "Transform (Modify) Bone - Bone: thigh_volume_l"(ComponentPose=@n20, Alpha=Get thigh_rot_l(), Translation="0,-1.379,1.177", Rotation="0,0,0", Scale="1,1,1") {
    translationMode=BMM_Additive
    translationSpace=BCS_BoneSpace
    alphaBoolBlend {blendOption=Linear}
    boneToModify {boneName=thigh_volume_l}
    translation {y=-1.379, z=1.177}
   }
   n22 = "Anim Dynamics - Chain: dyn_shoestring_a_1_r -> dyn_shoestring_a_2_r"(ComponentPose=@n21, Alpha=1, ExternalForce="0,0,0") {
    angularDampingOverride=0.995
    angularSpringConstant=300
    bAngularSpring=true
    bChain=true
    bGravityOverrideInSimSpace=true
    bOverrideAngularDamping=true
    bOverrideLinearDamping=true
    gravityScale=0.35
    linearDampingOverride=0.995
    simulationSpace=RootRelative
    alphaBoolBlend {blendOption=Linear}
    boundBone {boneName=dyn_shoestring_a_1_r}
    chainEnd {boneName=dyn_shoestring_a_2_r}
    componentAppliedLinearAccClamp=<3000 3000 3000>
    componentLinearAccScale=<0.2 0.2 0.2>
    componentLinearVelScale=<0.2 0.2 0.2>
    physicsBodyDefinitions [{
      sphereCollisionRadius=0
      boundBone {boneName=dyn_shoestring_a_1_r}
      boxExtents=<5 10 10>
      constraintSetup {angularTargetAxis=AxisZ, angularLimitsMax=<0 1 40>, angularLimitsMin=<0 -1 0>, angularTarget=<0 0 0>}
      localJointOffset=<2 0 0>
     }]
   }
   n23 = "Anim Dynamics - Chain: dyn_shoestring_b_1_r -> dyn_shoestring_b_2_r"(ComponentPose=@n22, Alpha=1, ExternalForce="0,0,0") {
    angularDampingOverride=0.995
    angularSpringConstant=300
    bAngularSpring=true
    bChain=true
    bGravityOverrideInSimSpace=true
    bOverrideAngularDamping=true
    bOverrideLinearDamping=true
    gravityScale=0.35
    linearDampingOverride=0.995
    simulationSpace=RootRelative
    alphaBoolBlend {blendOption=Linear}
    boundBone {boneName=dyn_shoestring_b_1_r}
    chainEnd {boneName=dyn_shoestring_b_2_r}
    componentAppliedLinearAccClamp=<1000 1000 1000>
    componentLinearAccScale=<0.1 0.1 0.1>
    componentLinearVelScale=<0.1 0.1 0.1>
    physicsBodyDefinitions [{
      sphereCollisionRadius=0
      boundBone {boneName=dyn_shoestring_b_1_r}
      boxExtents=<5 10 10>
      constraintSetup {angularTargetAxis=AxisZ, angularLimitsMax=<0 5 30>, angularLimitsMin=<0 -5 -5>, angularTarget=<0 0 0>}
      localJointOffset=<3 0 0>
     }]
   }
   n24 = Anim Dynamics - Bone: dyn_loleg_flap_a_l(ComponentPose=@n23, Alpha=1, ExternalForce="0,0,0") {
    angularDampingOverride=1
    angularSpringConstant=300
    bAngularSpring=true
    bGravityOverrideInSimSpace=true
    bOverrideAngularDamping=true
    bOverrideLinearDamping=true
    linearDampingOverride=0.9999
    simulationSpace=RootRelative
    alphaBoolBlend {blendOption=Linear}
    boundBone {boneName=dyn_loleg_flap_a_l}
    componentAppliedLinearAccClamp=<1000 1000 1000>
    componentLinearAccScale=<0.3 0.3 0.3>
    componentLinearVelScale=<0.3 0.3 0.3>
    physicsBodyDefinitions [{
      sphereCollisionRadius=0
      boundBone {boneName=dyn_loleg_flap_a_l}
      boxExtents=<5 10 10>
      constraintSetup {twistAxis=AxisZ, angularLimitsMax=<5 10 0>, angularLimitsMin=<-15 -10 0>, angularTarget=<0 0 0>}
      localJointOffset=<3 0 0>
     }]
   }
   n25 = Anim Dynamics - Bone: dyn_loleg_flap_b_l(ComponentPose=@n24, Alpha=1, ExternalForce="0,0,0") {
    angularDampingOverride=1
    angularSpringConstant=300
    bAngularSpring=true
    bGravityOverrideInSimSpace=true
    bOverrideAngularDamping=true
    bOverrideLinearDamping=true
    linearDampingOverride=0.9999
    simulationSpace=RootRelative
    alphaBoolBlend {blendOption=Linear}
    boundBone {boneName=dyn_loleg_flap_b_l}
    componentAppliedLinearAccClamp=<1000 1000 1000>
    componentLinearAccScale=<0.3 0.3 0.3>
    componentLinearVelScale=<0.3 0.3 0.3>
    physicsBodyDefinitions [{
      sphereCollisionRadius=0
      boundBone {boneName=dyn_loleg_flap_b_l}
      boxExtents=<5 10 10>
      constraintSetup {twistAxis=AxisZ, angularLimitsMax=<5 10 0>, angularLimitsMin=<-15 -10 0>, angularTarget=<0 0 0>}
      localJointOffset=<4 0 0>
     }]
   }
   Save cached pose 'CopyPoseBaseCache'(Pose=Component To Local(ComponentPose=@n25)) {cacheName=CopyPoseBaseCache}
   n26 = RigidBody(ComponentPose=Local To Component(LocalPose=Use cached pose 'CopyPoseBaseCache'())) {
    bClampLinearTranslationLimitToRefPose=true
    bOverrideWorldGravity=true
    overridePhysicsAsset=&/F_MED_Banner_BeltMain_Dyn_Physics
    simulationSpace=BaseBoneSpace
    alphaBoolBlend {blendOption=Linear}
    baseBoneRef {boneName=spine_01}
    componentAppliedLinearAccClamp=<3000 3000 3000>
    componentLinearAccScale=<0.1 0.1 0.4>
    componentLinearVelScale=<0.1 0.1 0.5>
    overrideWorldGravity {z=-500}
   }
   n27 = "Constraint - dyn_belt_main (P)"(ComponentPose=@n26, Alpha=1, ConstraintWeights_0=1) {
    alpha=0
    alphaBoolBlend {blendOption=Linear}
    alphaScaleBiasClamp {bInterpResult=true, interpSpeedDecreasing=1, interpSpeedIncreasing=20}
    boneToModify {boneName=dyn_belt_main}
    constraintSetup [{targetBone {boneName=pelvis}}]
    constraintWeights [1]
   }
   Save cached pose 'dynamicBeltPose'(Pose=Component To Local(ComponentPose=@n27)) {cacheName=dynamicBeltPose}
  ]}
 EventGraph {nodes [
   Event BlueprintUpdateAnimation()
   > n1 = Sequence!(execute) -> (then_0, then_1, then_2:Exec, then_3:Exec, then_4:Exec)
   n2 = GetOwnerSkeletalMeshComponent()
   n3 = BreakTransform(InTransform=GetDeltaTransformFromRefPose(self=@n2, BoneName=thigh_l, BaseName=pelvis))
   > Set thigh_rot_l(thigh_rot_l=MapRangeClamped(Value=BreakRotator(InRot=@n3.Rotation).Yaw, InRangeA=-20, InRangeB=-90))
   Set jump_alpha(execute=@n1.then_2, jump_alpha:double=MapRangeClamped(Value=Get PawnVelocityZ(), InRangeA=-700, InRangeB=0, OutRangeA=1, OutRangeB=0)) -> (Output_Get:double)
   n4 = BreakTransform(InTransform=GetDeltaTransformFromRefPose(self=@n2, BoneName=thigh_r, BaseName=pelvis))
   Set thigh_rot_r(execute=@n1.then_1, thigh_rot_r=MapRangeClamped(Value=BreakRotator(InRot=@n4.Rotation).Yaw, InRangeA=0, InRangeB=-105))
   n5 = BreakTransform(InTransform=GetDeltaTransformFromRefPose(self=@n2, BoneName=thigh_r, BaseName=pelvis))
   Set thigh_rotY_r(execute=@n1.then_3, thigh_rotY_r:double=MapRangeClamped(Value=BreakRotator(InRot=@n5.Rotation).Pitch, InRangeA=14, InRangeB=40)) -> (Output_Get:double)
   n6 = BreakTransform(InTransform=GetDeltaTransformFromRefPose(self=@n2, BoneName=calf_l, BaseName=thigh_l))
   Set kneepad_alpha(execute=@n1.then_4, kneepad_alpha:double=MapRangeClamped(Value=BreakRotator(InRot=@n6.Rotation).Yaw, InRangeA=20, InRangeB=115)) -> (Output_Get:double)
   Event OnBeginCrouching()
   > Set crouch_alpha(crouch_alpha:bool=true) -> (Output_Get:bool)
   Event OnEndCrouching()
   > Set crouch_alpha(crouch_alpha=false)
   TryGetPawnOwner()
  ]}
]
variables [
 crouch_alpha {type=boolean}
 jump_alpha {type=number}
 kneepad_alpha {type=number}
 thigh_rot_l {type=number}
 thigh_rot_r {type=number}
 thigh_rotY_r {type=number}
]
