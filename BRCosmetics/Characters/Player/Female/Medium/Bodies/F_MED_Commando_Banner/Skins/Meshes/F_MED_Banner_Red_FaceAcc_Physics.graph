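# Physics asset for the red Banner female commando skin's face accessory mesh, defining bodies for dynamic hair strands.
# Box bodies form three bone chains, long left, short left and right, each linked from a base through a parent to numbered segments by constraints with narrow swing and twist limits.
# Unconstrained head and neck capsules provide collision shapes, and no constraint ties the chain bases to them.
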PhysicsAsset /BRCosmetics/Characters/Player/Female/Medium/Bodies/F_MED_Commando_Banner/Skins/Meshes/F_MED_Banner_Red_FaceAcc_Physics
bodies {
 dyn_hair_long_l_01 [{shapeType=Box, extentX=2.4208, extentY=2.2395, extentZ=0.5003, center=<1.1954 0 -0.0001>}]
 dyn_hair_long_l_02 [{shapeType=Box, extentX=2.9495, extentY=1.8749, extentZ=1.0631, center=<1.599 0 -0.0001>}]
 dyn_hair_long_l_03 [{shapeType=Box, extentX=3.378, extentY=2.1664, extentZ=0.8545, center=<1.656 -0.0001 -0.2318>, rotation=<rot 0.0001 0 11.2943>}]
 dyn_hair_long_l_base [{shapeType=Box, extentX=0.8362, extentY=0.8362, extentZ=0.8362}]
 dyn_hair_long_l_parent [{shapeType=Box, extentX=2.7395, extentY=2.5921, extentZ=0.8731, center=<1.3896 0 0>}]
 dyn_hair_r_01 [{shapeType=Box, extentX=2.8861, extentY=2.9979, extentZ=1.0631, center=<1.4916 0 0>}]
 dyn_hair_r_02 [{shapeType=Box, extentX=4.3548, extentY=2.3327, extentZ=0.8437, center=<2.1615 0 0>}]
 dyn_hair_r_base [{shapeType=Box, extentX=0.7799, extentY=0.7799, extentZ=0.7799}]
 dyn_hair_r_parent [{shapeType=Box, extentX=3.8219, extentY=2.6723, extentZ=1.0631, center=<1.7227 0 0>}]
 dyn_hair_short_l_01 [{shapeType=Box, extentX=2.7547, extentY=1.3798, extentZ=0.7008, center=<1.36 0 -0.0001>}]
 dyn_hair_short_l_base [{shapeType=Box, extentX=0.8362, extentY=0.8362, extentZ=0.8362}]
 dyn_hair_short_l_parent [{shapeType=Box, extentX=1.941, extentY=1.2302, extentZ=0.741, center=<1.008 0 0>}]
 head [{shapeType=Capsule, length=4.5532, radius=6.0538, center=<0.3364 3.7485 0.0001>, rotation=<rot 90 90 71.8988>}]
 neck_02 [{shapeType=Capsule, length=5.5, radius=4.403, center=<0 0.4706 0>, rotation=<rot 90 0 0>}]
}
constraints [
 {bone1Name=dyn_hair_short_l_parent, bone2Name=dyn_hair_short_l_base, swing1LimitDegrees=10, swing2LimitDegrees=10, twistLimitDegrees=3}
 {bone1Name=dyn_hair_short_l_01, bone2Name=dyn_hair_short_l_parent, swing1LimitDegrees=10, swing2LimitDegrees=10, twistLimitDegrees=3}
 {bone1Name=dyn_hair_long_l_03, bone2Name=dyn_hair_long_l_02, swing1LimitDegrees=10, swing2LimitDegrees=20, twistLimitDegrees=3}
 {bone1Name=dyn_hair_long_l_parent, bone2Name=dyn_hair_long_l_base, swing1LimitDegrees=5, swing2LimitDegrees=10, twistLimitDegrees=3}
 {bone1Name=dyn_hair_long_l_01, bone2Name=dyn_hair_long_l_parent, swing1LimitDegrees=5, swing2LimitDegrees=10, twistLimitDegrees=3}
 {bone1Name=dyn_hair_long_l_02, bone2Name=dyn_hair_long_l_01, swing1LimitDegrees=10, swing2LimitDegrees=10, twistLimitDegrees=3}
 {bone1Name=dyn_hair_r_02, bone2Name=dyn_hair_r_01, swing1LimitDegrees=10, swing2LimitDegrees=20, twistLimitDegrees=3}
 {bone1Name=dyn_hair_r_parent, bone2Name=dyn_hair_r_base, swing1LimitDegrees=10, swing2LimitDegrees=20, twistLimitDegrees=3}
 {bone1Name=dyn_hair_r_01, bone2Name=dyn_hair_r_parent, swing1LimitDegrees=10, swing2LimitDegrees=20, twistLimitDegrees=3}
]
